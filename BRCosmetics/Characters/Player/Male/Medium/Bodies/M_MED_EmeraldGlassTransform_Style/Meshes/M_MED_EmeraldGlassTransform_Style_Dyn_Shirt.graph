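# Physics asset for the dynamic shirt of the male medium EmeraldGlassTransform Style body, driving its secondary motion.
# Capsules on the pelvis, spine, groin and upper thighs act as body collision for the shirt chains around front, back and sides.
# Short one- or two-segment chains with narrow angular limits hang from MSTR bones on spine_03, except left-front on pelvis.
# Angularly free constraints cross-link back-corner tips to side tips.
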PhysicsAsset /BRCosmetics/Characters/Player/Male/Medium/Bodies/M_MED_EmeraldGlassTransform_Style/Meshes/M_MED_EmeraldGlassTransform_Style_Dyn_Shirt
bodies {
 dyn_groin_collision [{shapeType=Capsule, length=15.475, radius=12.4622, center=<0 -4.9712 7.5752>, rotation=<rot 90 0 0>}]
 dyn_shirt_bk_01 [{shapeType=Capsule, length=17.3303, radius=8.0415, shapeName=dyn_jacket_l_01_capsule, center=<10 0 4>, rotation=<rot 0 0 90>}]
 dyn_shirt_bk_02 [{shapeType=Capsule, length=17.3303, radius=8.0415, shapeName=dyn_jacket_l_01_capsule, center=<10 0 4>, rotation=<rot 0 0 90>}]
 dyn_shirt_bk_MSTR [{shapeType=Capsule, length=1, radius=1}]
 dyn_shirt_bk_l_01 [{shapeType=Capsule, length=14.8303, radius=8.5415, shapeName=dyn_jacket_l_01_capsule, center=<10 0 6>, rotation=<rot 0 0 90>}]
 dyn_shirt_bk_l_02 [{shapeType=Capsule, length=14.5803, radius=8.7915, shapeName=dyn_jacket_l_01_capsule, center=<10 0 6>, rotation=<rot 0 0 90>}]
 dyn_shirt_bk_l_MSTR [{shapeType=Capsule, length=1, radius=1}]
 dyn_shirt_bk_r_01 [{shapeType=Capsule, length=16.0803, radius=8.5415, shapeName=dyn_jacket_l_01_capsule, center=<10 0 4>, rotation=<rot -0.0001 0 95>}]
 dyn_shirt_bk_r_02 [{shapeType=Capsule, length=16.0803, radius=8.5415, shapeName=dyn_jacket_l_01_capsule, center=<10 0 4>, rotation=<rot -0.0001 0 95>}]
 dyn_shirt_bk_r_MSTR [{shapeType=Capsule, length=1, radius=1}]
 dyn_shirt_ft_01 [{shapeType=Capsule, length=12.0803, radius=8.7915, shapeName=dyn_jacket_l_01_capsule, center=<10 0 8>, rotation=<rot 0.0006 -0.0002 105.0002>}]
 dyn_shirt_ft_MSTR [{shapeType=Capsule, length=1, radius=1}]
 dyn_shirt_ft_l_01 [{shapeType=Capsule, length=12.0803, radius=9.5415, shapeName=dyn_jacket_l_01_capsule, center=<13 0 6>, rotation=<rot 0.936 10.7541 85.0882>}]
 dyn_shirt_ft_l_MSTR [{shapeType=Capsule, length=1, radius=1}]
 dyn_shirt_ft_r_01 [{shapeType=Capsule, length=7.0803, radius=10.0415, shapeName=dyn_jacket_l_01_capsule, center=<10 0 7>, rotation=<rot 0.0002 -0.0002 95>}]
 dyn_shirt_ft_r_MSTR [{shapeType=Capsule, length=1, radius=1}]
 dyn_shirt_l_01 [{shapeType=Capsule, length=13.0803, radius=9.0415, shapeName=dyn_jacket_l_01_capsule, center=<6 0 6>, rotation=<rot 0 0 90>}]
 dyn_shirt_l_02 [{shapeType=Capsule, length=13.0803, radius=9.0415, shapeName=dyn_jacket_l_01_capsule, center=<6 0 7>, rotation=<rot 0 0 90>}]
 dyn_shirt_l_MSTR [{shapeType=Capsule, length=1, radius=1}]
 dyn_shirt_l_ft_01 [{shapeType=Capsule, length=11.0803, radius=10.5415, shapeName=dyn_jacket_l_01_capsule, center=<12 0 6>, rotation=<rot 0 0 85>}]
 dyn_shirt_l_ft_MSTR [{shapeType=Capsule, length=1, radius=1}]
 dyn_shirt_r_01 [{shapeType=Capsule, length=14.3303, radius=8.7915, shapeName=dyn_jacket_l_01_capsule, center=<6 0 6>, rotation=<rot 0 -0.0001 95>}]
 dyn_shirt_r_02 [{shapeType=Capsule, length=14.3303, radius=8.7915, shapeName=dyn_jacket_l_01_capsule, center=<6 0 6>, rotation=<rot 0 -0.0001 95>}]
 dyn_shirt_r_MSTR [{shapeType=Capsule, length=1, radius=1}]
 pelvis [{shapeType=Capsule, length=7.725, radius=13.4622, center=<-7.2714 -0.5323 0>}]
 spine_01 [{shapeType=Capsule, length=8.4006, radius=9.9006, center=<-1.5581 -1.1205 0>}]
 spine_02 [{shapeType=Capsule, length=8.9006, radius=8.4006, center=<0 -2.0742 0>}]
 spine_03 [{shapeType=Capsule, length=4.9006, radius=8.9006, center=<0 -1.3697 0>}]
 thigh_twist_01_l [{shapeType=Capsule, length=31.225, radius=9.9622, center=<-0.1305 -3.2542 0.8642>, rotation=<rot 90 179.7016 -177.2046>}]
 thigh_twist_01_r [{shapeType=Capsule, length=31.225, radius=9.9622, center=<0.254 4.8959 -0.8642>, rotation=<rot 90 179.1048 3.9691>}]
}
constraints [
 {
  bone1Name=dyn_shirt_bk_MSTR
  bone2Name=spine_03
  swing1LimitDegrees=45
  swing1Motion=Free
  swing2LimitDegrees=45
  swing2Motion=Free
  twistLimitDegrees=45
  twistMotion=Free
 }
 {bone1Name=dyn_shirt_bk_01, bone2Name=dyn_shirt_bk_MSTR, swing1LimitDegrees=15, swing2LimitDegrees=20, twistLimitDegrees=5}
 {bone1Name=dyn_shirt_bk_02, bone2Name=dyn_shirt_bk_01, swing1LimitDegrees=15, swing2LimitDegrees=20, twistLimitDegrees=5}
 {
  bone1Name=dyn_shirt_bk_l_MSTR
  bone2Name=spine_03
  swing1LimitDegrees=45
  swing1Motion=Free
  swing2LimitDegrees=45
  swing2Motion=Free
  twistLimitDegrees=45
  twistMotion=Free
 }
 {bone1Name=dyn_shirt_bk_l_01, bone2Name=dyn_shirt_bk_l_MSTR, swing1LimitDegrees=15, swing2LimitDegrees=20, twistLimitDegrees=5}
 {bone1Name=dyn_shirt_bk_l_02, bone2Name=dyn_shirt_bk_l_01, swing1LimitDegrees=15, swing2LimitDegrees=20, twistLimitDegrees=5}
 {
  bone1Name=dyn_shirt_bk_r_MSTR
  bone2Name=spine_03
  swing1LimitDegrees=45
  swing1Motion=Free
  swing2LimitDegrees=45
  swing2Motion=Free
  twistLimitDegrees=45
  twistMotion=Free
 }
 {bone1Name=dyn_shirt_bk_r_01, bone2Name=dyn_shirt_bk_r_MSTR, swing1LimitDegrees=15, swing2LimitDegrees=20, twistLimitDegrees=5}
 {bone1Name=dyn_shirt_bk_r_02, bone2Name=dyn_shirt_bk_r_01, swing1LimitDegrees=15, swing2LimitDegrees=20, twistLimitDegrees=5}
 {
  bone1Name=dyn_shirt_ft_MSTR
  bone2Name=spine_03
  swing1LimitDegrees=45
  swing1Motion=Free
  swing2LimitDegrees=45
  swing2Motion=Free
  twistLimitDegrees=45
  twistMotion=Free
 }
 {bone1Name=dyn_shirt_ft_01, bone2Name=dyn_shirt_ft_MSTR, swing1LimitDegrees=20, swing2LimitDegrees=20, twistLimitDegrees=5}
 {
  bone1Name=dyn_shirt_ft_l_MSTR
  bone2Name=spine_03
  swing1LimitDegrees=45
  swing1Motion=Free
  swing2LimitDegrees=45
  swing2Motion=Free
  twistLimitDegrees=45
  twistMotion=Free
 }
 {bone1Name=dyn_shirt_ft_l_01, bone2Name=dyn_shirt_ft_l_MSTR, swing1LimitDegrees=20, swing2LimitDegrees=20, twistLimitDegrees=5}
 {
  bone1Name=dyn_shirt_ft_r_MSTR
  bone2Name=spine_03
  swing1LimitDegrees=45
  swing1Motion=Free
  swing2LimitDegrees=45
  swing2Motion=Free
  twistLimitDegrees=45
  twistMotion=Free
 }
 {bone1Name=dyn_shirt_ft_r_01, bone2Name=dyn_shirt_ft_r_MSTR, swing1LimitDegrees=20, swing2LimitDegrees=20, twistLimitDegrees=2}
 {
  bone1Name=dyn_shirt_l_MSTR
  bone2Name=spine_03
  swing1LimitDegrees=45
  swing1Motion=Free
  swing2LimitDegrees=45
  swing2Motion=Free
  twistLimitDegrees=45
  twistMotion=Free
 }
 {bone1Name=dyn_shirt_l_01, bone2Name=dyn_shirt_l_MSTR, swing1LimitDegrees=20, swing2LimitDegrees=35, twistLimitDegrees=10}
 {bone1Name=dyn_shirt_l_02, bone2Name=dyn_shirt_l_01, swing1LimitDegrees=20, swing2LimitDegrees=35, twistLimitDegrees=10}
 {
  bone1Name=dyn_shirt_r_MSTR
  bone2Name=spine_03
  swing1LimitDegrees=45
  swing1Motion=Free
  swing2LimitDegrees=45
  swing2Motion=Free
  twistLimitDegrees=45
  twistMotion=Free
 }
 {bone1Name=dyn_shirt_r_01, bone2Name=dyn_shirt_r_MSTR, swing1LimitDegrees=25, swing2LimitDegrees=30, twistLimitDegrees=5}
 {bone1Name=dyn_shirt_r_02, bone2Name=dyn_shirt_r_01, swing1LimitDegrees=25, swing2LimitDegrees=30, twistLimitDegrees=5}
 {
  bone1Name=spine_01
  bone2Name=pelvis
  swing1LimitDegrees=45
  swing1Motion=Free
  swing2LimitDegrees=45
  swing2Motion=Free
  twistLimitDegrees=45
  twistMotion=Free
 }
 {
  bone1Name=spine_03
  bone2Name=spine_02
  swing1LimitDegrees=45
  swing1Motion=Free
  swing2LimitDegrees=45
  swing2Motion=Free
  twistLimitDegrees=45
  twistMotion=Free
 }
 {
  bone1Name=spine_02
  bone2Name=spine_01
  swing1LimitDegrees=45
  swing1Motion=Free
  swing2LimitDegrees=45
  swing2Motion=Free
  twistLimitDegrees=45
  twistMotion=Free
 }
 {
  bone1Name=dyn_groin_collision
  bone2Name=pelvis
  swing1LimitDegrees=45
  swing1Motion=Free
  swing2LimitDegrees=45
  swing2Motion=Free
  twistLimitDegrees=45
  twistMotion=Free
 }
 {
  bone1Name=dyn_shirt_l_ft_MSTR
  bone2Name=pelvis
  swing1LimitDegrees=45
  swing1Motion=Free
  swing2LimitDegrees=45
  swing2Motion=Free
  twistLimitDegrees=45
  twistMotion=Free
 }
 {
  bone1Name=dyn_shirt_bk_l_02
  bone2Name=dyn_shirt_l_02
  swing1LimitDegrees=45
  swing1Motion=Free
  swing2LimitDegrees=45
  swing2Motion=Free
  twistLimitDegrees=45
  twistMotion=Free
 }
 {bone1Name=dyn_shirt_l_ft_01, bone2Name=dyn_shirt_l_ft_MSTR, swing1LimitDegrees=20, swing2LimitDegrees=20, twistLimitDegrees=2}
 {
  bone1Name=thigh_twist_01_r
  bone2Name=pelvis
  swing1LimitDegrees=45
  swing1Motion=Free
  swing2LimitDegrees=45
  swing2Motion=Free
  twistLimitDegrees=45
  twistMotion=Free
 }
 {
  bone1Name=dyn_shirt_bk_r_02
  bone2Name=dyn_shirt_r_02
  swing1LimitDegrees=45
  swing1Motion=Free
  swing2LimitDegrees=45
  swing2Motion=Free
  twistLimitDegrees=45
  twistMotion=Free
 }
 {
  bone1Name=thigh_twist_01_l
  bone2Name=pelvis
  swing1LimitDegrees=45
  swing1Motion=Free
  swing2LimitDegrees=45
  swing2Motion=Free
  twistLimitDegrees=45
  twistMotion=Free
 }
]
