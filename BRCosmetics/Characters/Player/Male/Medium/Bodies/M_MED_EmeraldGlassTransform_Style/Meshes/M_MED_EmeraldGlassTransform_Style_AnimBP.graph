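# Post-process style AnimBP for the EmeraldGlassTransform male body cosmetic, adding secondary motion to shirt, sleeves and belt.
# It copies the parent mesh pose, runs twist and deform Control Rigs, AnimDynamics sleeves and attach-point constraints into a cached pose.
# Thigh-lift alphas drive corrective bone offsets, and bRigidBodyEnabled blends between a RigidBody shirt simulation and a constraint-only fallback.
# The event graph updates those alphas and a movement-state gravity override.
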AnimBlueprint /BRCosmetics/Characters/Player/Male/Medium/Bodies/M_MED_EmeraldGlassTransform_Style/Meshes/M_MED_EmeraldGlassTransform_Style_AnimBP
target_skeleton=&/M_MED_EmeraldGlassTransform_Style_Skeleton
graphs [
 AnimGraph {nodes [
   n1 = Get thigh_rot_up_l() -> (thigh_rot_up_l:double)
   n2 = "Transform (Modify) Bone - Bone: dyn_shirt_l_ft_MSTR"(ComponentPose=Local To Component(LocalPose=Use cached pose 'basePose'()), Alpha=@n1, Translation="-4,0,2", Rotation="85,0,0", Scale="1,1,1") {
    rotationMode=BMM_Additive
    rotationSpace=BCS_BoneSpace
    translationMode=BMM_Additive
    translationSpace=BCS_BoneSpace
    alphaBoolBlend {blendOption=Linear}
    alphaScaleBiasClamp {inRange {min=0.3}}
    boneToModify {boneName=dyn_shirt_l_ft_MSTR}
    rotation {pitch=85}
    translation {x=-4, z=2}
   }
   n3 = "Transform (Modify) Bone - Bone: dyn_shirt_ft_l_MSTR"(ComponentPose=@n2, Alpha=@n1, Translation="-2,0,2", Rotation="85,-20,20", Scale="1,1,1") {
    rotationMode=BMM_Additive
    rotationSpace=BCS_BoneSpace
    translationMode=BMM_Additive
    translationSpace=BCS_BoneSpace
    alphaBoolBlend {blendOption=Linear}
    alphaScaleBiasClamp {inRange {min=0.3}}
    boneToModify {boneName=dyn_shirt_ft_l_MSTR}
    rotation=<rot 85 -20 20>
    translation {x=-2, z=2}
   }
   n4 = "Transform (Modify) Bone - Bone: dyn_shirt_l_MSTR"(ComponentPose=@n3, Alpha=@n1, Translation="-3,-3,0", Rotation="30,25,-30", Scale="1,1,1") {
    rotationMode=BMM_Additive
    rotationSpace=BCS_BoneSpace
    translationMode=BMM_Additive
    translationSpace=BCS_BoneSpace
    alphaBoolBlend {blendOption=Linear}
    alphaScaleBiasClamp {inRange {min=0.3}}
    boneToModify {boneName=dyn_shirt_l_MSTR}
    rotation=<rot 30 25 -30>
    translation=<-3 -3>
   }
   n5 = Get thigh_rot_up_r() -> (thigh_rot_up_r:double)
   n6 = "Transform (Modify) Bone - Bone: dyn_shirt_ft_MSTR"(ComponentPose=@n4, Alpha=@n5, Translation="0,0,2", Rotation="60,0,0", Scale="1,1,1") {
    rotationMode=BMM_Additive
    rotationSpace=BCS_BoneSpace
    translationMode=BMM_Additive
    translationSpace=BCS_BoneSpace
    alphaBoolBlend {blendOption=Linear}
    boneToModify {boneName=dyn_shirt_ft_MSTR}
    rotation {pitch=60}
    translation {z=2}
   }
   n7 = "Transform (Modify) Bone - Bone: dyn_shirt_ft_r_MSTR"(ComponentPose=@n6, Alpha=@n5, Translation="0,0,2", Rotation="60,0,0", Scale="1,1,1") {
    rotationMode=BMM_Additive
    rotationSpace=BCS_BoneSpace
    translationMode=BMM_Additive
    translationSpace=BCS_BoneSpace
    alphaBoolBlend {blendOption=Linear}
    boneToModify {boneName=dyn_shirt_ft_r_MSTR}
    rotation {pitch=60}
    translation {z=2}
   }
   n8 = "Transform (Modify) Bone - Bone: dyn_shirt_r_MSTR"(ComponentPose=@n7, Alpha=@n5, Translation="0,0,0", Rotation="15,-25,0", Scale="1,1,1") {
    rotationMode=BMM_Additive
    rotationSpace=BCS_BoneSpace
    translationMode=BMM_Additive
    translationSpace=BCS_BoneSpace
    alphaBoolBlend {blendOption=Linear}
    boneToModify {boneName=dyn_shirt_r_MSTR}
    rotation {pitch=15, yaw=-25}
   }
   n9 = "Constraint - dyn_groin_collision (PP)"!(ComponentPose=@n8, Alpha=1, ConstraintWeights_0=0.5, ConstraintWeights_1:float=0.5) -> (Pose) {
    alphaBoolBlend {blendOption=Linear}
    boneToModify {boneName=dyn_groin_collision}
    constraintSetup [{targetBone {boneName=thigh_twist_01_l}}, {targetBone {boneName=thigh_twist_01_r}}]
    constraintWeights [1,1]
   }
   n10 = "Transform (Modify) Bone - Bone: dyn_groin_collision"(ComponentPose=@n9, Alpha=Get thigh_rot_up_l(), Translation="0,0,0", Rotation="0,45,0", Scale="1,1,1") {
    rotationMode=BMM_Additive
    rotationSpace=BCS_BoneSpace
    translationMode=BMM_Additive
    translationSpace=BCS_BoneSpace
    alphaBoolBlend {blendOption=Linear}
    boneToModify {boneName=dyn_groin_collision}
    rotation {yaw=45}
   }
   n11 = "Transform (Modify) Bone - Bone: dyn_groin_collision"(ComponentPose=@n10, Alpha=Get thigh_rot_up_r(), Translation="0,0,0", Rotation="0,-45,0", Scale="1,1,1") {
    rotationMode=BMM_Additive
    rotationSpace=BCS_BoneSpace
    translationMode=BMM_Additive
    translationSpace=BCS_BoneSpace
    alphaBoolBlend {blendOption=Linear}
    boneToModify {boneName=dyn_groin_collision}
    rotation {yaw=-45}
   }
   n12 = Look At - Bone: dyn_shirt_bk_MSTR(ComponentPose=@n11, Alpha=1) {
    alphaBoolBlend {blendOption=Linear}
    boneToModify {boneName=dyn_shirt_bk_MSTR}
    lookAtLocation=<0 -16.7965 -0.9635>
    lookAtTarget {boneReference {boneName=spine_02}}
    lookAt_Axis {axis=<1 0>}
   }
   n13 = Look At - Bone: dyn_shirt_bk_l_MSTR(ComponentPose=@n12, Alpha=1) {
    alphaBoolBlend {blendOption=Linear}
    boneToModify {boneName=dyn_shirt_bk_l_MSTR}
    lookAtLocation=<0 -13.0025 -11.9746>
    lookAtTarget {boneReference {boneName=spine_02}}
    lookAt_Axis {axis=<1 0>}
   }
   n14 = Look At - Bone: dyn_shirt_bk_r_MSTR(ComponentPose=@n13, Alpha=1) {
    alphaBoolBlend {blendOption=Linear}
    boneToModify {boneName=dyn_shirt_bk_r_MSTR}
    lookAtLocation=<0 -13.6207 12.5231>
    lookAtTarget {boneReference {boneName=spine_02}}
    lookAt_Axis {axis=<1 0>}
   }
   n15 = RigidBody!(ComponentPose=@n14, Alpha, OverrideWorldGravity:Vector=Get GravityOverride(), ExternalForce) -> (Pose) {
    bClampLinearTranslationLimitToRefPose=true
    bOverrideWorldGravity=true
    overridePhysicsAsset=&/M_MED_EmeraldGlassTransform_Style_Dyn_Shirt
    simulationSpace=BaseBoneSpace
    alphaBoolBlend {blendOption=Linear}
    baseBoneRef {boneName=spine_01}
    componentAppliedLinearAccClamp=<3000 1000 1000>
    overrideWorldGravity {z=-980}
    simSpaceSettings {maxAngularAcceleration=50, maxAngularVelocity=2, maxLinearAcceleration=300, maxLinearVelocity=100, velocityScaleZ=0, worldAlpha=1}
   }
   n16 = Get thigh_rot_up_l()
   n17 = "Transform (Modify) Bone - Bone: dyn_shirt_l_ft_MSTR"(ComponentPose=@n15, Alpha=@n16, Translation="0,0,-5", Rotation="0.000,0.010,-15.000", Scale="0.4,1,1") {
    rotationMode=BMM_Additive
    rotationSpace=BCS_BoneSpace
    scaleMode=BMM_Additive
    scaleSpace=BCS_BoneSpace
    translationMode=BMM_Additive
    translationSpace=BCS_BoneSpace
    alphaBoolBlend {blendOption=Linear}
    alphaScaleBiasClamp {inRange {min=0.3}}
    boneToModify {boneName=dyn_shirt_l_ft_MSTR}
    rotation=<rot 0 0.01 -15>
    scale {x=0.4}
    translation {z=-5}
   }
   n18 = "Transform (Modify) Bone - Bone: dyn_shirt_ft_l_MSTR"(ComponentPose=@n17, Alpha=@n16, Translation="0,0,-2", Rotation="0,0,0", Scale="0.35,1,1") {
    rotationMode=BMM_Additive
    rotationSpace=BCS_BoneSpace
    scaleMode=BMM_Additive
    scaleSpace=BCS_BoneSpace
    translationMode=BMM_Additive
    translationSpace=BCS_BoneSpace
    alphaBoolBlend {blendOption=Linear}
    alphaScaleBiasClamp {inRange {min=0.3}}
    boneToModify {boneName=dyn_shirt_ft_l_MSTR}
    scale {x=0.35}
    translation {z=-2}
   }
   n19 = Get thigh_rot_up_r()
   n20 = "Transform (Modify) Bone - Bone: dyn_shirt_ft_MSTR"(ComponentPose=@n18, Alpha=@n19, Translation="0,0,-2", Rotation="0,0,0", Scale="0.4,1,1") {
    rotationMode=BMM_Additive
    rotationSpace=BCS_BoneSpace
    scaleMode=BMM_Additive
    scaleSpace=BCS_BoneSpace
    translationMode=BMM_Additive
    translationSpace=BCS_BoneSpace
    alphaBoolBlend {blendOption=Linear}
    alphaScaleBiasClamp {inRange {min=0.5}}
    boneToModify {boneName=dyn_shirt_ft_MSTR}
    scale {x=0.4}
    translation {z=-2}
   }
   n21 = "Transform (Modify) Bone - Bone: dyn_shirt_ft_r_MSTR"(ComponentPose=@n20, Alpha=@n19, Translation="0,0,-2", Rotation="0,0,0", Scale="0.4,1,1") {
    rotationMode=BMM_Additive
    rotationSpace=BCS_BoneSpace
    scaleMode=BMM_Additive
    scaleSpace=BCS_BoneSpace
    translationMode=BMM_Additive
    translationSpace=BCS_BoneSpace
    alphaBoolBlend {blendOption=Linear}
    alphaScaleBiasClamp {inRange {min=0.5}}
    boneToModify {boneName=dyn_shirt_ft_r_MSTR}
    scale {x=0.4}
    translation {z=-2}
   }
   n22 = Look At - Bone: dyn_shirt_bk_MSTR(ComponentPose=Local To Component(LocalPose=Use cached pose 'basePose'()), Alpha=1) {
    alphaBoolBlend {blendOption=Linear}
    boneToModify {boneName=dyn_shirt_bk_MSTR}
    lookAtLocation=<0 -16.7965 -0.9635>
    lookAtTarget {boneReference {boneName=spine_02}}
    lookAt_Axis {axis=<1 0>}
   }
   n23 = Look At - Bone: dyn_shirt_bk_l_MSTR(ComponentPose=@n22, Alpha=1) {
    alphaBoolBlend {blendOption=Linear}
    boneToModify {boneName=dyn_shirt_bk_l_MSTR}
    lookAtLocation=<0 -13.0025 -11.9746>
    lookAtTarget {boneReference {boneName=spine_02}}
    lookAt_Axis {axis=<1 0>}
   }
   n24 = Look At - Bone: dyn_shirt_bk_r_MSTR(ComponentPose=@n23, Alpha=1) {
    alphaBoolBlend {blendOption=Linear}
    boneToModify {boneName=dyn_shirt_bk_r_MSTR}
    lookAtLocation=<0 -13.6207 12.5231>
    lookAtTarget {boneReference {boneName=spine_02}}
    lookAt_Axis {axis=<1 0>}
   }
   n25 = "Constraint - dyn_shirt_bk_02 (P)"(ComponentPose=@n24, Alpha=1, ConstraintWeights_0=1) {
    alphaBoolBlend {blendOption=Linear}
    boneToModify {boneName=dyn_shirt_bk_02}
    constraintSetup [{targetBone {boneName=pelvis}}]
    constraintWeights [1]
   }
   n26 = "Constraint - dyn_shirt_bk_l_02 (P)"(ComponentPose=@n25, Alpha=1, ConstraintWeights_0=1) {
    alphaBoolBlend {blendOption=Linear}
    boneToModify {boneName=dyn_shirt_bk_l_02}
    constraintSetup [{targetBone {boneName=pelvis}}]
    constraintWeights [1]
   }
   n27 = "Constraint - dyn_shirt_bk_r_02 (P)"(ComponentPose=@n26, Alpha=1, ConstraintWeights_0=1) {
    alphaBoolBlend {blendOption=Linear}
    boneToModify {boneName=dyn_shirt_bk_r_02}
    constraintSetup [{targetBone {boneName=pelvis}}]
    constraintWeights [1]
   }
   n28 = "Constraint - dyn_shirt_l_02 (P)"(ComponentPose=@n27, Alpha=1, ConstraintWeights_0=1) {
    alphaBoolBlend {blendOption=Linear}
    boneToModify {boneName=dyn_shirt_l_02}
    constraintSetup [{targetBone {boneName=pelvis}}]
    constraintWeights [1]
   }
   n29 = "Constraint - dyn_shirt_r_02 (P)"(ComponentPose=@n28, Alpha=1, ConstraintWeights_0=1) {
    alphaBoolBlend {blendOption=Linear}
    boneToModify {boneName=dyn_shirt_r_02}
    constraintSetup [{targetBone {boneName=pelvis}}]
    constraintWeights [1]
   }
   n30 = Get thigh_rot_up_l()
   n31 = "Transform (Modify) Bone - Bone: dyn_shirt_l_ft_MSTR"(ComponentPose=@n29, Alpha=@n30, Translation="0,0,0", Rotation="48.907,6.559,-17.276", Scale="0.5,1,1") {
    rotationMode=BMM_Additive
    rotationSpace=BCS_BoneSpace
    scaleMode=BMM_Additive
    scaleSpace=BCS_BoneSpace
    translationMode=BMM_Additive
    translationSpace=BCS_BoneSpace
    alphaBoolBlend {blendOption=Linear}
    boneToModify {boneName=dyn_shirt_l_ft_MSTR}
    rotation=<rot 48.907 6.559 -17.276>
    scale {x=0.5}
   }
   n32 = "Transform (Modify) Bone - Bone: dyn_shirt_ft_l_MSTR"(ComponentPose=@n31, Alpha=@n30, Translation="0,0,0", Rotation="52.867,-26.550,17.753", Scale="0.5,1,1") {
    rotationMode=BMM_Additive
    rotationSpace=BCS_BoneSpace
    scaleMode=BMM_Additive
    scaleSpace=BCS_BoneSpace
    translationMode=BMM_Additive
    translationSpace=BCS_BoneSpace
    alphaBoolBlend {blendOption=Linear}
    boneToModify {boneName=dyn_shirt_ft_l_MSTR}
    rotation=<rot 52.867 -26.55 17.753>
    scale {x=0.5}
   }
   n33 = "Transform (Modify) Bone - Bone: dyn_shirt_l_02"(ComponentPose=@n32, Alpha=@n30, Translation="-0.164,-0.105,0.775", Rotation="14.085,30.647,-4.761", Scale="1,1,1") {
    rotationMode=BMM_Additive
    rotationSpace=BCS_BoneSpace
    translationMode=BMM_Additive
    translationSpace=BCS_BoneSpace
    alphaBoolBlend {blendOption=Linear}
    boneToModify {boneName=dyn_shirt_l_02}
    rotation=<rot 14.085 30.647 -4.761>
    translation=<-0.164 -0.105 0.775>
   }
   n34 = Get thigh_rot_up_r()
   n35 = "Transform (Modify) Bone - Bone: dyn_shirt_ft_MSTR"(ComponentPose=@n33, Alpha=@n34, Translation="0,0,0", Rotation="50,0,0", Scale="0.5,1,1") {
    rotationMode=BMM_Additive
    rotationSpace=BCS_BoneSpace
    scaleMode=BMM_Additive
    scaleSpace=BCS_BoneSpace
    translationMode=BMM_Additive
    translationSpace=BCS_BoneSpace
    alphaBoolBlend {blendOption=Linear}
    boneToModify {boneName=dyn_shirt_ft_MSTR}
    rotation {pitch=50}
    scale {x=0.5}
   }
   n36 = "Transform (Modify) Bone - Bone: dyn_shirt_ft_r_MSTR"(ComponentPose=@n35, Alpha=@n34, Translation="0,0,0", Rotation="50,0,0", Scale="0.5,1,1") {
    rotationMode=BMM_Additive
    rotationSpace=BCS_BoneSpace
    scaleMode=BMM_Additive
    scaleSpace=BCS_BoneSpace
    translationMode=BMM_Additive
    translationSpace=BCS_BoneSpace
    alphaBoolBlend {blendOption=Linear}
    boneToModify {boneName=dyn_shirt_ft_r_MSTR}
    rotation {pitch=50}
    scale {x=0.5}
   }
   n37 = "Transform (Modify) Bone - Bone: dyn_shirt_r_02"(ComponentPose=@n36, Alpha=@n34, Translation="0,0,0", Rotation="15,-30,5", Scale="1,1,1") {
    rotationMode=BMM_Additive
    rotationSpace=BCS_BoneSpace
    translationMode=BMM_Additive
    translationSpace=BCS_BoneSpace
    alphaBoolBlend {blendOption=Linear}
    boneToModify {boneName=dyn_shirt_r_02}
    rotation=<rot 15 -30 5>
   }
   Output Pose(Result=Blend Poses by bool(BlendPose_0=Component To Local(ComponentPose=@n21), BlendPose_1=Component To Local(ComponentPose=@n37), bActiveValue=Get bRigidBodyEnabled()) {blendType=Linear}) {layerGroup=None}
   n38 = Control Rig(Source=Copy Pose From Mesh() {bUseAttachedParent=true}) {
    alphaBoolBlend {blendOption=Linear}
    controlRigAssetReference {blueprintRigClass=&/M_MED_EmeraldGlassTransform_Style_CtrlRig_Twist.M_MED_EmeraldGlassTransform_Style_CtrlRig_Twist_C}
   }
   n39 = Control Rig(Source=@n38) {
    alphaBoolBlend {blendOption=Linear}
    controlRigAssetReference {blueprintRigClass=/Game/Characters/Player/Male/Medium/Bodies/M_MED_EmeraldGlassTransform_Style/Meshes/M_MED_EmeraldGlassTransform_Style_DeformRig_ControlRig.M_MED_EmeraldGlassTransform_Style_DeformRig_ControlRig_C}
   }
   n40 = Anim Dynamics - Bone: dyn_sleeve_dn_r!(ComponentPose=Local To Component(LocalPose=@n39), Alpha=1, GravityScale:float=1, ExternalForce="0,0,0") -> (Pose) {
    angularDampingOverride=0.99
    angularSpringConstant=1000
    bAngularSpring=true
    bGravityOverrideInSimSpace=true
    bOverrideAngularDamping=true
    bUseGravityOverride=true
    simulationSpace=RootRelative
    alphaBoolBlend {blendOption=Linear}
    boundBone {boneName=dyn_sleeve_dn_r}
    componentAppliedLinearAccClamp=<3000 3000 3000>
    componentLinearAccScale=<0.2 0.2 0.2>
    componentLinearVelScale=<0.3 0.3 0.3>
    gravityOverride {z=-500}
    physicsBodyDefinitions [{
      sphereCollisionRadius=0
      boundBone {boneName=dyn_sleeve_dn_r}
      boxExtents=<20 20 10>
      constraintSetup {angularLimitsMax=<3 2 10>, angularLimitsMin=<-3 -15 -10>, angularTarget=<0 0 0>}
      localJointOffset=<5 0 0>
     }]
    relativeSpaceBone {boneName=lowerarm_r}
   }
   n41 = Anim Dynamics - Bone: dyn_sleeve_ft_r!(ComponentPose=@n40, Alpha=1, GravityScale:float=1, ExternalForce="0,0,0") -> (Pose) {
    angularDampingOverride=0.9999
    angularSpringConstant=500
    bGravityOverrideInSimSpace=true
    bOverrideAngularDamping=true
    bUseGravityOverride=true
    simulationSpace=RootRelative
    alphaBoolBlend {blendOption=Linear}
    boundBone {boneName=dyn_sleeve_ft_r}
    componentAppliedLinearAccClamp=<3000 3000 3000>
    componentLinearAccScale=<0.1 0.1 0.1>
    componentLinearVelScale=<0.3 0.3 0.3>
    gravityOverride {z=-500}
    physicsBodyDefinitions [{
      sphereCollisionRadius=0
      boundBone {boneName=dyn_sleeve_ft_r}
      boxExtents=<10 20 20>
      constraintSetup {angularLimitsMax=<3 5 2>, angularLimitsMin=<-3 -5 -2>, angularTarget=<0 0 0>}
      localJointOffset=<5 1 2>
     }]
    relativeSpaceBone {boneName=upperarm_r}
   }
   n42 = Anim Dynamics - Bone: dyn_sleeve_bk_r!(ComponentPose=@n41, Alpha=1, GravityScale:float=1, ExternalForce="0,0,0") -> (Pose) {
    angularDampingOverride=0.9999
    angularSpringConstant=500
    bGravityOverrideInSimSpace=true
    bOverrideAngularDamping=true
    bUseGravityOverride=true
    simulationSpace=RootRelative
    alphaBoolBlend {blendOption=Linear}
    boundBone {boneName=dyn_sleeve_bk_r}
    componentAppliedLinearAccClamp=<3000 3000 3000>
    componentLinearAccScale=<0.1 0.1 0.1>
    componentLinearVelScale=<0.3 0.3 0.3>
    gravityOverride {z=-500}
    physicsBodyDefinitions [{
      sphereCollisionRadius=0
      boundBone {boneName=dyn_sleeve_bk_r}
      boxExtents=<10 20 20>
      constraintSetup {angularLimitsMax=<3 5 2>, angularLimitsMin=<-3 -5 -2>, angularTarget=<0 0 0>}
      localJointOffset=<5 -1 2>
     }]
    relativeSpaceBone {boneName=upperarm_r}
   }
   n43 = Anim Dynamics - Bone: dyn_sleeve_up_r!(ComponentPose=@n42, Alpha=1, GravityScale:float=1, ExternalForce="0,0,0") -> (Pose) {
    angularDampingOverride=1
    angularSpringConstant=500
    bGravityOverrideInSimSpace=true
    bOverrideAngularDamping=true
    bUseGravityOverride=true
    simulationSpace=BoneRelative
    alphaBoolBlend {blendOption=Linear}
    boundBone {boneName=dyn_sleeve_up_r}
    componentAppliedLinearAccClamp=<3000 3000 3000>
    componentLinearAccScale=<0.1 0.1 0.1>
    componentLinearVelScale=<0.2 0.2 0.2>
    gravityOverride {y=500, z=50}
    physicsBodyDefinitions [{
      sphereCollisionRadius=0
      boundBone {boneName=dyn_sleeve_up_r}
      boxExtents=<15 20 20>
      constraintSetup {angularLimitsMax=<0 2 5>, angularLimitsMin=<0 -5 -5>, angularTarget=<0 0 0>}
      localJointOffset=<5 0 0>
     }]
    relativeSpaceBone {boneName=upperarm_twist_01_r}
   }
   n44 = Anim Dynamics - Bone: dyn_sleeve_dn_l!(ComponentPose=@n43, Alpha=1, GravityScale:float=1, ExternalForce="0,0,0") -> (Pose) {
    angularDampingOverride=0.99
    angularSpringConstant=1000
    bAngularSpring=true
    bGravityOverrideInSimSpace=true
    bOverrideAngularDamping=true
    bUseGravityOverride=true
    simulationSpace=RootRelative
    alphaBoolBlend {blendOption=Linear}
    boundBone {boneName=dyn_sleeve_dn_l}
    componentAppliedLinearAccClamp=<3000 3000 3000>
    componentLinearAccScale=<0.2 0.2 0.2>
    componentLinearVelScale=<0.3 0.3 0.3>
    gravityOverride {z=-500}
    physicsBodyDefinitions [{
      sphereCollisionRadius=0
      boundBone {boneName=dyn_sleeve_dn_l}
      boxExtents=<20 20 10>
      constraintSetup {angularLimitsMax=<3 2 10>, angularLimitsMin=<-3 -15 -10>, angularTarget=<0 0 0>}
      localJointOffset=<5 0 0>
     }]
    relativeSpaceBone {boneName=lowerarm_r}
   }
   n45 = Anim Dynamics - Bone: dyn_sleeve_ft_l!(ComponentPose=@n44, Alpha=1, GravityScale:float=1, ExternalForce="0,0,0") -> (Pose) {
    angularDampingOverride=0.9999
    angularSpringConstant=500
    bGravityOverrideInSimSpace=true
    bOverrideAngularDamping=true
    bUseGravityOverride=true
    simulationSpace=RootRelative
    alphaBoolBlend {blendOption=Linear}
    boundBone {boneName=dyn_sleeve_ft_l}
    componentAppliedLinearAccClamp=<3000 3000 3000>
    componentLinearAccScale=<0.1 0.1 0.1>
    componentLinearVelScale=<0.3 0.3 0.3>
    gravityOverride {z=-500}
    physicsBodyDefinitions [{
      sphereCollisionRadius=0
      boundBone {boneName=dyn_sleeve_ft_l}
      boxExtents=<10 20 20>
      constraintSetup {angularLimitsMax=<3 5 2>, angularLimitsMin=<-3 -5 -2>, angularTarget=<0 0 0>}
      localJointOffset=<5 1 2>
     }]
    relativeSpaceBone {boneName=upperarm_r}
   }
   n46 = Anim Dynamics - Bone: dyn_sleeve_bk_l!(ComponentPose=@n45, Alpha=1, GravityScale:float=1, ExternalForce="0,0,0") -> (Pose) {
    angularDampingOverride=0.9999
    angularSpringConstant=500
    bGravityOverrideInSimSpace=true
    bOverrideAngularDamping=true
    bUseGravityOverride=true
    simulationSpace=RootRelative
    alphaBoolBlend {blendOption=Linear}
    boundBone {boneName=dyn_sleeve_bk_l}
    componentAppliedLinearAccClamp=<3000 3000 3000>
    componentLinearAccScale=<0.1 0.1 0.1>
    componentLinearVelScale=<0.3 0.3 0.3>
    gravityOverride {z=-500}
    physicsBodyDefinitions [{
      sphereCollisionRadius=0
      boundBone {boneName=dyn_sleeve_bk_l}
      boxExtents=<10 20 20>
      constraintSetup {angularLimitsMax=<3 5 2>, angularLimitsMin=<-3 -5 -2>, angularTarget=<0 0 0>}
      localJointOffset=<5 -1 2>
     }]
    relativeSpaceBone {boneName=upperarm_r}
   }
   n47 = Anim Dynamics - Bone: dyn_sleeve_up_l!(ComponentPose=@n46, Alpha=1, GravityScale:float=1, ExternalForce="0,0,0") -> (Pose) {
    angularDampingOverride=1
    angularSpringConstant=500
    bGravityOverrideInSimSpace=true
    bOverrideAngularDamping=true
    bUseGravityOverride=true
    simulationSpace=BoneRelative
    alphaBoolBlend {blendOption=Linear}
    boundBone {boneName=dyn_sleeve_up_l}
    componentAppliedLinearAccClamp=<3000 3000 3000>
    componentLinearAccScale=<0.1 0.1 0.1>
    componentLinearVelScale=<0.2 0.2 0.2>
    gravityOverride {y=500, z=-50}
    physicsBodyDefinitions [{
      sphereCollisionRadius=0
      boundBone {boneName=dyn_sleeve_up_l}
      boxExtents=<15 20 20>
      constraintSetup {angularLimitsMax=<0 2 5>, angularLimitsMin=<0 -5 -5>, angularTarget=<0 0 0>}
      localJointOffset=<5 0 0>
     }]
    relativeSpaceBone {boneName=upperarm_twist_01_l}
   }
   n48 = "Constraint - attach_backpack (PP)"!(ComponentPose=@n47, Alpha=1, ConstraintWeights_0=0.32, ConstraintWeights_1:float=0.68) -> (Pose) {
    alphaBoolBlend {blendOption=Linear}
    boneToModify {boneName=attach_backpack}
    constraintSetup [{targetBone {boneName=spine_04}}, {targetBone {boneName=spine_05}}]
    constraintWeights [1,1]
   }
   n49 = "Constraint - attach_cape (PP)"!(ComponentPose=@n48, Alpha=1, ConstraintWeights_0=0.06, ConstraintWeights_1:float=0.94) -> (Pose) {
    alphaBoolBlend {blendOption=Linear}
    boneToModify {boneName=attach_cape}
    constraintSetup [{targetBone {boneName=spine_04}}, {targetBone {boneName=spine_05}}]
    constraintWeights [1,1]
   }
   n50 = "Constraint - def_belt_buckle_MSTR (PPPPP)"!(
    ComponentPose=@n49
    Alpha=1
    ConstraintWeights_0=0.3174
    ConstraintWeights_1:float=0.2194
    ConstraintWeights_2:float=0.1796
    ConstraintWeights_3:float=0.1407
    ConstraintWeights_4:float=0.1429
   ) -> (Pose) {
    alphaBoolBlend {blendOption=Linear}
    boneToModify {boneName=def_belt_buckle_MSTR}
    constraintSetup [
     {targetBone {boneName=pelvis}}
     {targetBone {boneName=thigh_twist_01_l}}
     {targetBone {boneName=thigh_twist_01_r}}
     {targetBone {boneName=deform_groin_l}}
     {targetBone {boneName=deform_groin_r}}
    ]
    constraintWeights [1,1,1,1,1]
   }
   n51 = "Transform (Modify) Bone - Bone: def_belt_buckle"(ComponentPose=@n50, Alpha=Get thigh_rot_up_l(), Translation="1,2,0.5", Rotation="0,-5,30", Scale="1,1,1") {
    rotationMode=BMM_Additive
    rotationSpace=BCS_BoneSpace
    translationMode=BMM_Additive
    translationSpace=BCS_BoneSpace
    alphaBoolBlend {blendOption=Linear}
    boneToModify {boneName=def_belt_buckle}
    rotation {roll=30, yaw=-5}
    translation=<1 2 0.5>
   }
   n52 = "Transform (Modify) Bone - Bone: def_belt_buckle"(ComponentPose=@n51, Alpha=Get thigh_rot_up_r(), Translation="1,-2,0.5", Rotation="0,5,-30", Scale="1,1,1") {
    rotationMode=BMM_Additive
    rotationSpace=BCS_BoneSpace
    translationMode=BMM_Additive
    translationSpace=BCS_BoneSpace
    alphaBoolBlend {blendOption=Linear}
    boneToModify {boneName=def_belt_buckle}
    rotation {roll=-30, yaw=5}
    translation=<1 -2 0.5>
   }
   Save cached pose 'basePose'(Pose=Component To Local(ComponentPose=@n52)) {cacheName=basePose}
  ]}
 EventGraph {nodes [
   Event BlueprintUpdateAnimation()
   > n1 = Sequence()
   n2 = Get SkelMesh()
   > Set thigh_rot_up_l(thigh_rot_up_l:double=GetAlphaFromReferenceTransformDelta(SkeletalMeshComponent=@n2, BoneName=thigh_l, BaseName=pelvis, DeltaComponentType=QuaternionTwist, InputRange=(X=0,Y=110), TwistAxis="0,0,1")) -> (Output_Get:double)
   > Set thigh_rot_up_r(thigh_rot_up_r:double=GetAlphaFromReferenceTransformDelta(SkeletalMeshComponent=@n2, BoneName=thigh_r, BaseName=pelvis, DeltaComponentType=QuaternionTwist, InputRange=(X=0,Y=100), TwistAxis="0,0,1")) -> (Output_Get:double)
   > Set knee_rot_up_l(knee_rot_up_l:double=GetAlphaFromReferenceTransformDelta(SkeletalMeshComponent=@n2, BoneName=thigh_l, BaseName=calf_l, DeltaComponentType=QuaternionTwist, InputRange=(X=0,Y=100), TwistAxis="0,0,1")) -> (Output_Get:double)
   n3 = Gravity Override State Function DEPRECATED(
    execute=@n1.then_1
    CharacterPartAnimInstance=Self-Reference()
    GameTimeInSeconds=GetGameTimeInSeconds()
    GravityOverrideParams=Get Gravity Override Params()
   )
   > Set GravityOverride(GravityOverride=@n3)
   Event BlueprintInitializeAnimation()
   n4 = GetOwnerSkeletalMeshComponent()
   > Is Valid(InputObject=@n4)
   > Set SkelMesh(SkelMesh=@n4)
  ]}
]
variables [
 Gravity Override Params {
  type=GravityOverrideParamsStruct
  default {
   emoteOrMelee {gravityOverride=<0 0 -500>}
   front_End {gravityOverride=<0 0 -500>}
   onGround_Falling {gravityOverride=<0 0 -500>}
   onGround_Moving {gravityOverride=<0 0 -500>}
   onGround_Standing {gravityOverride=<0 0 -500>}
   skydive_FloatingDown {gravityOverride=<0 -700 -100>, windAmplitude=<200 -500 200>, windOffset=<0 -600 100>}
   skydive_GoingUp {gravityOverride=<0 -300 500>, windAmplitude=<200 -500 300>}
   skydive_Parachuting {gravityOverride=<0 -500 -980>, windAmplitude=<200 -150 400>, windFrequency=<1.5 4 6>}
   ziplining {gravityOverride=<0 0 -500>}
  }
 }
 GravityOverride {type=Vector}
 knee_rot_up_l {type=number}
 SkelMesh {type=SkeletalMeshComponent}
 thigh_rot_up_l {type=number}
 thigh_rot_up_r {type=number}
]
