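# Physics asset for the dynamic hair on the EmeraldGlassTransform_Style medium male face accessory, apparently the variant used while emoting.
# Left and right four-bone hair chains and two back tips hang from small MSTR anchor bodies on the head, while inner and outer tips constrain directly to it.
# Hair joints have tight swing and twist limits.
# Head, upper spine and arm capsules provide body shapes, joined like the MSTR anchors by free angular constraints.
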PhysicsAsset /Game/Characters/Player/Male/Medium/Bodies/M_MED_EmeraldGlassTransform_Style/Meshes/Parts/M_MED_EmeraldGlassTransform_Style_FaceAcc_Dyn_Emoting
bodies {
 dyn_hairTips_bk_01 [{shapeType=Capsule, length=3.75, radius=6.25, shapeName=dyn_hair_bk_01_capsule, center=<1 0 0>, rotation=<rot 90 0 0>}]
 dyn_hairTips_bk_01_MSTR [{shapeType=Capsule, length=0.505, radius=0.505, shapeName=dyn_hairTips_bk_01_MSTR_capsule, rotation=<rot 0 -90 0>}]
 dyn_hairTips_bk_02 [{shapeType=Capsule, length=3.75, radius=6.25, shapeName=dyn_hair_bk_01_capsule, center=<3 0 0>, rotation=<rot 90 0 0>}]
 dyn_hairTips_bk_02_MSTR [{shapeType=Capsule, length=0.505, radius=0.505, shapeName=dyn_hairTips_bk_02_MSTR_capsule, rotation=<rot 0 -90 0>}]
 dyn_hairTips_inr_01 [{shapeType=Capsule, length=3.75, radius=6.25, shapeName=dyn_hair_bk_01_capsule, center=<3 0 0>, rotation=<rot 90 0 0>}]
 dyn_hairTips_inr_02 [{shapeType=Capsule, length=3.75, radius=6.25, shapeName=dyn_hair_bk_01_capsule, center=<2 0 0>, rotation=<rot 90 0 0>}]
 dyn_hairTips_inr_03 [{shapeType=Capsule, length=3.75, radius=6.25, shapeName=dyn_hair_bk_01_capsule, center=<2 0 0>, rotation=<rot 90 0 0>}]
 dyn_hairTips_inr_04 [{shapeType=Capsule, length=3.75, radius=6.25, shapeName=dyn_hair_bk_01_capsule, center=<2 0 0>, rotation=<rot 90 0 0>}]
 dyn_hairTips_inr_05 [{shapeType=Capsule, length=3.75, radius=6.25, shapeName=dyn_hair_bk_01_capsule, center=<2 0 0>, rotation=<rot 90 0 0>}]
 dyn_hairTips_otr_01 [{shapeType=Capsule, length=3.75, radius=6.25, shapeName=dyn_hair_bk_01_capsule, center=<4 0 0>, rotation=<rot 90 0 0>}]
 dyn_hairTips_otr_02 [{shapeType=Capsule, length=3.75, radius=6.25, shapeName=dyn_hair_bk_01_capsule, center=<4 0 0>, rotation=<rot 90 0 0>}]
 dyn_hairTips_otr_03 [{shapeType=Capsule, length=3.75, radius=6.25, shapeName=dyn_hair_bk_01_capsule, center=<4 0 0>, rotation=<rot 90 0 0>}]
 dyn_hairTips_otr_04 [{shapeType=Capsule, length=3.75, radius=6.25, shapeName=dyn_hair_bk_01_capsule, center=<4 0 0>, rotation=<rot 90 0 0>}]
 dyn_hairTips_otr_05 [{shapeType=Capsule, length=3.75, radius=6.25, shapeName=dyn_hair_bk_01_capsule, center=<4 0 0>, rotation=<rot 90 0 0>}]
 dyn_hair_l_01 [{shapeType=Capsule, length=3, radius=2.375, center=<5 0 0>, rotation=<rot 90 0 0>}]
 dyn_hair_l_02 [{shapeType=Capsule, length=5.25, radius=2.625, center=<5 0 0>, rotation=<rot 90 0 0>}]
 dyn_hair_l_03 [{shapeType=Capsule, length=6, radius=1.875, center=<5 0 0>, rotation=<rot 90 0 0>}]
 dyn_hair_l_04 [{shapeType=Capsule, length=5.5, radius=3.375, center=<9 0.3477 0>, rotation=<rot 90 180 174.9997>}]
 dyn_hair_l_MSTR [{shapeType=Capsule, length=0.505, radius=0.505, shapeName=dyn_hair_l_MSTR_capsule, rotation=<rot 0 -90 0>}]
 dyn_hair_r_01 [{shapeType=Capsule, length=3, radius=2.375, center=<5 0 0>, rotation=<rot 90 0 0>}]
 dyn_hair_r_02 [{shapeType=Capsule, length=5.25, radius=2.625, center=<5 0 0>, rotation=<rot 90 0 0>}]
 dyn_hair_r_03 [{shapeType=Capsule, length=6, radius=1.875, center=<5 0 0>, rotation=<rot 90 0 0>}]
 dyn_hair_r_04 [{shapeType=Capsule, length=5.25, radius=3.125, center=<9 0 0>, rotation=<rot 90 0 0>}]
 dyn_hair_r_MSTR [{shapeType=Capsule, length=0.505, radius=0.505, shapeName=dyn_hair_r_MSTR_capsule, rotation=<rot 0 -90 0>}]
 head [{shapeType=Capsule, length=6.25, radius=14.5676, shapeName=head_capsule, center=<6.3911 -4.1928 0>, rotation=<rot 90 180 175.0001>}]
 lowerarm_l [{shapeType=Capsule, length=26, radius=4.875, center=<15.5239 0 -0.3086>, rotation=<rot 85 -0.0006 -175.0007>}]
 lowerarm_r [{shapeType=Capsule, length=26, radius=4.875, center=<-15.5239 0 0.3086>, rotation=<rot 85 -0.0006 4.9993>}]
 spine_05 [{shapeType=Capsule, length=13.25, radius=11.125, center=<0.3119 -2.2276 -0.0001>}]
 upperarm_l [{shapeType=Capsule, length=17, radius=8.125, center=<8.8661 -2.4461 -3.0668>, rotation=<rot 85 -179.9989 -4.999>}]
 upperarm_r [{shapeType=Capsule, length=17, radius=8.125, center=<-8.8661 2.4461 3.0668>, rotation=<rot 85 -179.9989 175.001>}]
}
constraints [
 {bone1Name=dyn_hairTips_bk_01, bone2Name=dyn_hairTips_bk_01_MSTR, swing1LimitDegrees=15, swing2LimitDegrees=15, twistLimitDegrees=5}
 {bone1Name=dyn_hairTips_bk_02, bone2Name=dyn_hairTips_bk_02_MSTR, swing1LimitDegrees=15, swing2LimitDegrees=15, twistLimitDegrees=5}
 {bone1Name=dyn_hair_l_01, bone2Name=dyn_hair_l_MSTR, swing1LimitDegrees=10, swing2LimitDegrees=15, twistLimitDegrees=5}
 {bone1Name=dyn_hair_l_02, bone2Name=dyn_hair_l_01, swing1LimitDegrees=15, swing2LimitDegrees=5, twistLimitDegrees=5}
 {bone1Name=dyn_hair_l_03, bone2Name=dyn_hair_l_02, swing1LimitDegrees=30, swing2LimitDegrees=30, twistLimitDegrees=5}
 {bone1Name=dyn_hair_l_04, bone2Name=dyn_hair_l_03, swing1LimitDegrees=30, swing2LimitDegrees=30, twistLimitDegrees=5}
 {bone1Name=dyn_hair_r_01, bone2Name=dyn_hair_r_MSTR, swing1LimitDegrees=10, swing2LimitDegrees=15, twistLimitDegrees=5}
 {bone1Name=dyn_hair_r_02, bone2Name=dyn_hair_r_01, swing1LimitDegrees=15, swing2LimitDegrees=5, twistLimitDegrees=5}
 {bone1Name=dyn_hair_r_03, bone2Name=dyn_hair_r_02, swing1LimitDegrees=20, swing2LimitDegrees=20, twistLimitDegrees=5}
 {bone1Name=dyn_hair_r_04, bone2Name=dyn_hair_r_03, swing1LimitDegrees=30, swing2LimitDegrees=30, twistLimitDegrees=5}
 {
  bone1Name=dyn_hair_r_MSTR
  bone2Name=head
  swing1LimitDegrees=45
  swing1Motion=Free
  swing2LimitDegrees=45
  swing2Motion=Free
  twistLimitDegrees=45
  twistMotion=Free
 }
 {
  bone1Name=dyn_hair_l_MSTR
  bone2Name=head
  swing1LimitDegrees=45
  swing1Motion=Free
  swing2LimitDegrees=45
  swing2Motion=Free
  twistLimitDegrees=45
  twistMotion=Free
 }
 {bone1Name=dyn_hairTips_inr_01, bone2Name=head, swing1LimitDegrees=20, swing2LimitDegrees=20, twistLimitDegrees=5}
 {bone1Name=dyn_hairTips_inr_02, bone2Name=head, swing1LimitDegrees=20, swing2LimitDegrees=20, twistLimitDegrees=5}
 {bone1Name=dyn_hairTips_inr_03, bone2Name=head, swing1LimitDegrees=20, swing2LimitDegrees=20, twistLimitDegrees=5}
 {bone1Name=dyn_hairTips_inr_04, bone2Name=head, swing1LimitDegrees=20, swing2LimitDegrees=20, twistLimitDegrees=5}
 {bone1Name=dyn_hairTips_inr_05, bone2Name=head, swing1LimitDegrees=20, swing2LimitDegrees=20, twistLimitDegrees=5}
 {bone1Name=dyn_hairTips_otr_01, bone2Name=head, swing1LimitDegrees=20, swing2LimitDegrees=20, twistLimitDegrees=5}
 {bone1Name=dyn_hairTips_otr_02, bone2Name=head, swing1LimitDegrees=20, swing2LimitDegrees=20, twistLimitDegrees=5}
 {bone1Name=dyn_hairTips_otr_03, bone2Name=head, swing1LimitDegrees=20, swing2LimitDegrees=20, twistLimitDegrees=5}
 {bone1Name=dyn_hairTips_otr_04, bone2Name=head, swing1LimitDegrees=20, swing2LimitDegrees=20, twistLimitDegrees=5}
 {bone1Name=dyn_hairTips_otr_05, bone2Name=head, swing1LimitDegrees=20, swing2LimitDegrees=20, twistLimitDegrees=5}
 {
  bone1Name=dyn_hairTips_bk_01_MSTR
  bone2Name=head
  swing1LimitDegrees=45
  swing1Motion=Free
  swing2LimitDegrees=45
  swing2Motion=Free
  twistLimitDegrees=45
  twistMotion=Free
 }
 {
  bone1Name=dyn_hairTips_bk_02_MSTR
  bone2Name=head
  swing1LimitDegrees=45
  swing1Motion=Free
  swing2LimitDegrees=45
  swing2Motion=Free
  twistLimitDegrees=45
  twistMotion=Free
 }
 {
  bone1Name=head
  bone2Name=spine_05
  swing1LimitDegrees=45
  swing1Motion=Free
  swing2LimitDegrees=45
  swing2Motion=Free
  twistLimitDegrees=45
  twistMotion=Free
 }
 {
  bone1Name=upperarm_r
  bone2Name=spine_05
  swing1LimitDegrees=45
  swing1Motion=Free
  swing2LimitDegrees=45
  swing2Motion=Free
  twistLimitDegrees=45
  twistMotion=Free
 }
 {
  bone1Name=lowerarm_r
  bone2Name=upperarm_r
  swing1LimitDegrees=45
  swing1Motion=Free
  swing2LimitDegrees=45
  swing2Motion=Free
  twistLimitDegrees=45
  twistMotion=Free
 }
 {
  bone1Name=upperarm_l
  bone2Name=spine_05
  swing1LimitDegrees=45
  swing1Motion=Free
  swing2LimitDegrees=45
  swing2Motion=Free
  twistLimitDegrees=45
  twistMotion=Free
 }
 {
  bone1Name=lowerarm_l
  bone2Name=upperarm_l
  swing1LimitDegrees=45
  swing1Motion=Free
  swing2LimitDegrees=45
  swing2Motion=Free
  twistLimitDegrees=45
  twistMotion=Free
 }
]
