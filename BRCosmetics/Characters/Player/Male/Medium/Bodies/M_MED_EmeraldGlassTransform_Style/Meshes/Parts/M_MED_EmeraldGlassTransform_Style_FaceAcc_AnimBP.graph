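# Post-copy AnimBP for the Emerald Glass Transform medium male face accessory mesh, animating its dynamic hair bones.
# A cached base pose blends attached-parent and owner-mesh copies at the upper arms, then aims back hair tips at the neck.
# With rigid body enabled, head-space simulation switches physics assets for emotes, adding bone offsets for the Scribe emote.
# Otherwise, look-ats and additive rotations, weighted by head and right-arm twist, pose the hair procedurally.
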AnimBlueprint /BRCosmetics/Characters/Player/Male/Medium/Bodies/M_MED_EmeraldGlassTransform_Style/Meshes/Parts/M_MED_EmeraldGlassTransform_Style_FaceAcc_AnimBP
target_skeleton=/Game/Characters/Player/Male/Medium/Bodies/M_MED_EmeraldGlassTransform_Style/Meshes/Parts/M_MED_EmeraldGlassTransform_Style_FaceAcc_Skeleton
graphs [
 AnimGraph {nodes [
   n1 = Look At - Bone: dyn_hair_r_MSTR(ComponentPose=Local To Component(LocalPose=Use cached pose 'basePose'()), Alpha=1) {
    alphaBoolBlend {blendOption=Linear}
    boneToModify {boneName=dyn_hair_r_MSTR}
    lookAtLocation=<-20 16.5 20>
    lookAtTarget {boneReference {boneName=spine_05}}
    lookAt_Axis {axis=<1 0>}
   }
   n2 = Look At - Bone: dyn_hair_l_MSTR(ComponentPose=@n1, Alpha=1) {
    alphaBoolBlend {blendOption=Linear}
    boneToModify {boneName=dyn_hair_l_MSTR}
    lookAtLocation=<-20 16.5 -20>
    lookAtTarget {boneReference {boneName=spine_05}}
    lookAt_Axis {axis=<1 0>}
   }
   n3 = RigidBody!(ComponentPose=@n2, Alpha, OverrideWorldGravity:Vector="0,0,-600", ExternalForce, ComponentLinearAccScale:Vector="0.2,0.15,0.15") -> (Pose) {
    bClampLinearTranslationLimitToRefPose=true
    bOverrideWorldGravity=true
    overridePhysicsAsset=/Game/Characters/Player/Male/Medium/Bodies/M_MED_EmeraldGlassTransform_Style/Meshes/Parts/M_MED_EmeraldGlassTransform_Style_FaceAcc_Dyn_Hair
    simulationSpace=BaseBoneSpace
    alphaBoolBlend {blendOption=Linear}
    alphaScaleBiasClamp {interpSpeedDecreasing=4, interpSpeedIncreasing=6}
    baseBoneRef {boneName=head}
    componentAppliedLinearAccClamp=<1000 1000 1000>
    componentLinearAccScale=<0.6 0.6 0.2>
    componentLinearVelScale=<0.35 0.35 0.35>
    overrideWorldGravity {z=-875}
    simSpaceSettings {maxAngularAcceleration=200, maxAngularVelocity=5, maxLinearAcceleration=500, maxLinearVelocity=200}
   }
   n4 = Component To Local(ComponentPose=@n3)
   n5 = RigidBody(
    ComponentPose=Local To Component(LocalPose=Use cached pose 'basePose'())
    OverrideWorldGravity="0,0,-600"
    ComponentLinearAccScale="0.1,0.1,0.1"
   ) {
    bClampLinearTranslationLimitToRefPose=true
    bOverrideWorldGravity=true
    overridePhysicsAsset=/Game/Characters/Player/Male/Medium/Bodies/M_MED_EmeraldGlassTransform_Style/Meshes/Parts/M_MED_EmeraldGlassTransform_Style_FaceAcc_Dyn_Emoting
    simulationSpace=BaseBoneSpace
    alphaBoolBlend {blendOption=Linear}
    alphaScaleBiasClamp {interpSpeedDecreasing=4, interpSpeedIncreasing=6}
    baseBoneRef {boneName=head}
    componentAppliedLinearAccClamp=<1000 1000 1000>
    componentLinearAccScale=<0.6 0.6 0.2>
    componentLinearVelScale=<0.35 0.35 0.35>
    overrideWorldGravity {z=-875}
    simSpaceSettings {maxAngularAcceleration=200, maxAngularVelocity=5, maxLinearAcceleration=500, maxLinearVelocity=200}
   }
   n6 = Get isPlayingEmoteScribe() -> (isPlayingEmoteScribe:bool)
   n7 = "Transform (Modify) Bone - Bone: dyn_hair_l_03"!(ComponentPose=@n5, bAlphaBoolEnabled:bool=@n6, Translation="0,0,0", Rotation="-30,-20,0", Scale="1,1,1") -> (Pose) {
    alphaInputType=Bool
    rotationMode=BMM_Additive
    rotationSpace=BCS_BoneSpace
    translationMode=BMM_Additive
    translationSpace=BCS_BoneSpace
    alphaBoolBlend {blendInTime=0.3, blendOption=Linear}
    boneToModify {boneName=dyn_hair_l_03}
    rotation {pitch=-30, yaw=-20}
   }
   n8 = "Transform (Modify) Bone - Bone: dyn_hair_l_04"!(ComponentPose=@n7, bAlphaBoolEnabled:bool=@n6, Translation="0,0,0", Rotation="-15,40,0", Scale="1,1,1") -> (Pose) {
    alphaInputType=Bool
    rotationMode=BMM_Additive
    rotationSpace=BCS_BoneSpace
    translationMode=BMM_Additive
    translationSpace=BCS_BoneSpace
    alphaBoolBlend {blendInTime=0.3, blendOption=Linear}
    boneToModify {boneName=dyn_hair_l_04}
    rotation {pitch=-15, yaw=40}
   }
   n9 = Component To Local(ComponentPose=@n8)
   n10 = Get bIsPlayingEmote()
   n11 = Look At - Bone: dyn_hair_r_MSTR(ComponentPose=Local To Component(LocalPose=Use cached pose 'basePose'()), Alpha=1) {
    alphaBoolBlend {blendOption=Linear}
    boneToModify {boneName=dyn_hair_r_MSTR}
    lookAtLocation=<-20 16.5 20>
    lookAtTarget {boneReference {boneName=spine_05}}
    lookAt_Axis {axis=<1 0>}
   }
   n12 = Look At - Bone: dyn_hair_l_MSTR(ComponentPose=@n11, Alpha=1) {
    alphaBoolBlend {blendOption=Linear}
    boneToModify {boneName=dyn_hair_l_MSTR}
    lookAtLocation=<-20 16.5 -20>
    lookAtTarget {boneReference {boneName=spine_05}}
    lookAt_Axis {axis=<1 0>}
   }
   n13 = Get headRotateUp() -> (headRotateUp:double)
   n14 = "Transform (Modify) Bone - Bone: dyn_hair_r_01"(ComponentPose=@n12, Alpha=@n13, Translation="0,0,0", Rotation="20,-20,0", Scale="1,1,1") {
    rotationMode=BMM_Additive
    rotationSpace=BCS_BoneSpace
    translationMode=BMM_Additive
    translationSpace=BCS_BoneSpace
    alphaBoolBlend {blendOption=Linear}
    boneToModify {boneName=dyn_hair_r_01}
    rotation {pitch=20, yaw=-20}
   }
   n15 = "Transform (Modify) Bone - Bone: dyn_hair_l_01"(ComponentPose=@n14, Alpha=@n13, Translation="0,0,0", Rotation="20,20,0", Scale="1,1,1") {
    rotationMode=BMM_Additive
    rotationSpace=BCS_BoneSpace
    translationMode=BMM_Additive
    translationSpace=BCS_BoneSpace
    alphaBoolBlend {blendOption=Linear}
    boneToModify {boneName=dyn_hair_l_01}
    rotation {pitch=20, yaw=20}
   }
   n16 = Look At - Bone: dyn_hair_r_03(ComponentPose=@n15, Alpha=1) {
    alphaBoolBlend {blendOption=Linear}
    boneToModify {boneName=dyn_hair_r_03}
    lookAtLocation=<-20 -8 -10>
    lookAtTarget {boneReference {boneName=spine_05}}
    lookAt_Axis {axis=<1 0>}
   }
   n17 = Look At - Bone: dyn_hair_l_03(ComponentPose=@n16, Alpha=1) {
    alphaBoolBlend {blendOption=Linear}
    boneToModify {boneName=dyn_hair_l_03}
    lookAtLocation=<-20 8 10>
    lookAtTarget {boneReference {boneName=spine_05}}
    lookAt_Axis {axis=<1 0>}
   }
   n18 = Get rightArmRotateUp() -> (rightArmRotateUp:double)
   n19 = "Transform (Modify) Bone - Bone: dyn_hair_r_01"(ComponentPose=@n17, Alpha=@n18, Translation="0,0,0", Rotation="-22.556,-17.917,-9.159", Scale="1,1,1") {
    rotationMode=BMM_Additive
    rotationSpace=BCS_BoneSpace
    translationMode=BMM_Additive
    translationSpace=BCS_BoneSpace
    alphaBoolBlend {blendOption=Linear}
    alphaScaleBiasClamp {bInterpResult=true}
    boneToModify {boneName=dyn_hair_r_01}
    rotation=<rot -22.556 -17.9171 -9.1589>
   }
   n20 = "Transform (Modify) Bone - Bone: dyn_hair_r_03"(ComponentPose=@n19, Alpha=@n18, Translation="0,0,0", Rotation="9.877,-0.071,1.718", Scale="1,1,1") {
    rotationMode=BMM_Additive
    rotationSpace=BCS_BoneSpace
    translationMode=BMM_Additive
    translationSpace=BCS_BoneSpace
    alphaBoolBlend {blendOption=Linear}
    alphaScaleBiasClamp {bInterpResult=true}
    boneToModify {boneName=dyn_hair_r_03}
    rotation=<rot 9.877 -0.071 1.718>
   }
   n21 = "Transform (Modify) Bone - Bone: dyn_hair_r_04"(ComponentPose=@n20, Alpha=@n18, Translation="0,0,0", Rotation="-24.207,-48.555,-6.776", Scale="1,1,1") {
    rotationMode=BMM_Additive
    rotationSpace=BCS_BoneSpace
    translationMode=BMM_Additive
    translationSpace=BCS_BoneSpace
    alphaBoolBlend {blendOption=Linear}
    alphaScaleBiasClamp {bInterpResult=true}
    boneToModify {boneName=dyn_hair_r_04}
    rotation=<rot -24.2069 -48.5551 -6.776>
   }
   Output Pose(Result=Blend Poses by bool(
     BlendPose_0=Blend Poses by bool(BlendPose_0=@n9, BlendPose_1=@n4, BlendTime_0=0.3, bActiveValue=@n10) {blendType=Linear}
     BlendPose_1=Component To Local(ComponentPose=@n21)
     bActiveValue=Get bRigidBodyEnabled()
    ) {blendType=Linear}) {layerGroup=None}
   n22 = Copy Pose From Mesh(SourceMeshComponent=Get SkelMesh())
   n23 = Copy Pose From Mesh() {bUseAttachedParent=true}
   n24 = Look At - Bone: dyn_hairTips_bk_01_MSTR(
    ComponentPose=Local To Component(LocalPose=Layered blend per bone(BasePose=@n23, BlendPoses_0=@n22) {layerSetup [{branchFilters [{boneName=upperarm_l}, {boneName=upperarm_r}]}]})
    Alpha=1
   ) {
    alphaBoolBlend {blendOption=Linear}
    boneToModify {boneName=dyn_hairTips_bk_01_MSTR}
    lookAtLocation=<-12 -10 -15>
    lookAtTarget {boneReference {boneName=neck_01}}
    lookAt_Axis {axis=<1 0>}
   }
   n25 = Look At - Bone: dyn_hairTips_bk_02_MSTR(ComponentPose=@n24, Alpha=1) {
    alphaBoolBlend {blendOption=Linear}
    boneToModify {boneName=dyn_hairTips_bk_02_MSTR}
    lookAtLocation=<-12 -10 14>
    lookAtTarget {boneReference {boneName=neck_01}}
    lookAt_Axis {axis=<1 0>}
   }
   Save cached pose 'basePose'(Pose=Component To Local(ComponentPose=@n25)) {cacheName=basePose}
  ]}
 EventGraph {nodes [
   Event BlueprintUpdateAnimation()
   > n1 = Sequence()
   n2 = Get SkelMesh()
   > Set rightArmRotateUp(rightArmRotateUp:double=GetAlphaFromReferenceTransformDelta(
     SkeletalMeshComponent=@n2
     BoneName=upperarm_r
     BaseName=clavicle_r
     DeltaComponentType=QuaternionTwist
     InputRange=(X=0,Y=-100)
     TwistAxis="1,0,0"
    )) -> (Output_Get:double)
   > Set headRotateDown(headRotateDown:double=GetAlphaFromReferenceTransformDelta(SkeletalMeshComponent=@n2, BoneName=head, BaseName=spine_05, DeltaComponentType=QuaternionTwist, InputRange=(X=0,Y=-20), TwistAxis="0,0,1")) -> (Output_Get:double)
   > Set headRotateUp(headRotateUp:double=GetAlphaFromReferenceTransformDelta(SkeletalMeshComponent=@n2, BoneName=head, BaseName=spine_05, DeltaComponentType=QuaternionTwist, InputRange=(X=0,Y=45), TwistAxis="0,0,1")) -> (Output_Get:double)
   n3 = GetAnimInstance(self=Get SkelMesh())
   Is Valid(
    exec=@n1.then_1
    InputObject=GetCurrentActiveMontage!(self:Anim Instance Object Reference=@n3) -> (ReturnValue:Anim Montage Object Reference)
   )
   > Set isPlayingEmoteScribe(isPlayingEmoteScribe:bool=Montage_IsPlaying(self=@n3, Montage=/Game/Animation/Game/MainPlayer/Emotes/Scribe/Emote_Scribe_CMM_M)) -> (Output_Get:bool)
   Event BlueprintInitializeAnimation()
   n4 = GetOwnerSkeletalMeshComponent()
   > Is Valid(InputObject=@n4)
   > Set SkelMesh(SkelMesh=@n4)
  ]}
]
variables [
 headRotateDown {type=number}
 headRotateUp {type=number}
 isPlayingEmoteScribe {type=boolean}
 rightArmRotateUp {type=number}
 SkelMesh {type=SkeletalMeshComponent}
]
